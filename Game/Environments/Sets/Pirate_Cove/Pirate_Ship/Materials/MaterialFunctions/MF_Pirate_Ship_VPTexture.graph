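# Pirate Ship material function that blends a vertex-painted texture layer over incoming material attributes, behind a static switch.
# Inverted red vertex color and per-material spread drive two height lerps against a background height map: one for the paint, one for a darkened outline.
# Scale, spread and normal flattening vary across up to four sub-materials, with per-texture visibility keyed by transitionFactor.
# A switch tints with Athena snow color; height mask and color are extra outputs.
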MaterialFunction /Game/Environments/Sets/Pirate_Cove/Pirate_Ship/Materials/MaterialFunctions/MF_Pirate_Ship_VPTexture
graph {nodes [
  n1 = TextureCoordinate()
  n2 = ScalarParameter() {defaultValue=1, group=Material_1, parameterName=VP_TextureScale, sortPriority=15}
  n3 = ScalarParameter() {defaultValue=1, group=Material_2, parameterName=VP_TextureScale2, sortPriority=15}
  n4 = ScalarParameter() {defaultValue=1, group=Material_3, parameterName=VP_TextureScale3, sortPriority=15}
  n5 = ScalarParameter() {defaultValue=1, group=Material_4, parameterName=VP_TextureScale4, sortPriority=15}
  n6 = StaticBoolParameter() {group=Material_2, parameterName=Use 2 Materials}
  n7 = StaticBoolParameter() {group=Material_3, parameterName=Use 3 Materials}
  n8 = StaticBoolParameter() {group=Material_4, parameterName=Use 4 Materials}
  n9 = Multiply(
   A=@n1
   B=MaterialFunctionCall!(
    transitionFactor
    1=Ceil(None=@n2)
    2=Ceil(None=@n3)
    3=Ceil(None=@n4)
    4=Ceil(None=@n5)
    Use 2 Textures=@n6
    Use 3 Textures=@n7
    Use 4 Textures=@n8
   ) {materialFunction=/Game/Athena/WinterArt/Materials/MF_LerpTexturesBasedOnInt_Float4_4Layers}.Result
  )
  n10 = TextureObjectParameter() {
   group=VertexPaintTexture
   parameterName=VP_Color
   sortPriority=15
   texture=/Game/Environments/Sets/Pirate_Cove/Pirate_Ship/Textures/T_Pirate_Ship_Hull_B_D
  }
  n11 = TextureSample(UVs=@n9, Tex=@n10) {samplerType=SAMPLERTYPE_Normal, texture=/Game/Global/Textures/Default/Tileable_Patterns/T_Snow01_N}
  n12 = AppendVector(A=@n11.RGB, B=@n11.RGB)
  ComponentMask(None=@n12) {r=true}
  FunctionInput() {inputName=1, inputType=FunctionInput_Vector4}
  FunctionInput() {inputName=2, inputType=FunctionInput_Vector4, sortPriority=1}
  FunctionInput() {inputName=3, inputType=FunctionInput_Vector4, sortPriority=2}
  FunctionInput() {inputName=4, inputType=FunctionInput_Vector4, sortPriority=3}
  n13 = FunctionInput() {inputType=FunctionInput_MaterialAttributes}
  n14 = GetMaterialAttributes(None=@n13) {attributeGetTypes [
    #1
    #2
    #3
    #4
    #5
    #6
   ]}
  n15 = VectorParameter() {group=VertexPaintTexture, parameterName=VP_OutlineD_Mod, sortPriority=13, defaultValue=<rgba 0.7 0.7 0.7 0>}
  n16 = Multiply(A=@n14.BaseColor, B=@n15.RGB) {constB=0.5}
  n17 = VertexColor()
  n18 = Constant() {r=3}
  n19 = FunctionInput(Preview=ComponentMask(None=TextureCoordinate() {coordinateIndex=1}) {r=true}) {bUsePreviewValueAsDefault=true, inputName=transitionFactor, inputType=FunctionInput_Scalar, sortPriority=-1}
  n20 = If(
   A=Constant() {r=2}
   B=@n19
   "A > B"=If(
    A=Constant() {r=1}
    B=@n19
    "A > B"=ScalarParameter() {group=VertexPaintTexture, parameterName=Show VP on Texture 1?}
    "A < B"=ScalarParameter() {group=VertexPaintTexture, parameterName=Show VP on Texture 2?}
   ) {desc=0}
   "A < B"=ScalarParameter() {group=VertexPaintTexture, parameterName=Show VP on Texture 3?}
  )
  n21 = ScalarParameter() {group=VertexPaintTexture, parameterName=Show VP on Texture 4?}
  n22 = Multiply(
   A=MaterialFunctionCall(
    1=ScalarParameter() {defaultValue=1, group=Material_1, parameterName=VP_Spread_1, sliderMax=2, sortPriority=11}
    2=ScalarParameter() {defaultValue=1, group=Material_2, parameterName=VP_Spread_2, sliderMax=2, sortPriority=11}
    3=ScalarParameter() {defaultValue=1, group=Material_3, parameterName=VP_Spread_3, sliderMax=2, sortPriority=11}
    4=ScalarParameter() {defaultValue=1, group=Material_4, parameterName=VP_Spread_4, sliderMax=2, sortPriority=11}
    Use 2 Textures=@n6
    Use 3 Textures=@n7
    Use 4 Textures=@n8
   ) {materialFunction=&MF_LerpTexturesBasedOnInt_Float4_4Layers}.Result
   B=LinearInterpolate(A=Constant() {r=1}, B=OneMinus(None=@n17.R), Alpha=If(A=@n18, B=@n19, "A > B"=@n20, "A < B"=@n21))
  )
  n23 = Constant()
  n24 = Constant() {r=0.35}
  n25 = ScalarParameter() {defaultValue=0.7, group=VertexPaintTexture, parameterName=OutlineDistance, sortPriority=4}
  n26 = LinearInterpolate(
   A=ScalarParameter() {group=VertexPaintTexture, parameterName=VP_Background_HM_Min, sortPriority=2}
   B=ScalarParameter() {defaultValue=1, group=VertexPaintTexture, parameterName=VP_Background_HM_Max, sortPriority=1}
   Alpha=FunctionInput() {bUsePreviewValueAsDefault=true, inputName=BGMap, inputType=FunctionInput_Scalar, sortPriority=1}
  )
  n27 = ComponentMask(None=@n12) {a=true}
  n28 = Multiply(A=@n27, B=ScalarParameter() {defaultValue=1, group=VertexPaintTexture, parameterName=VP_TextureHeightMax, sortPriority=3})
  n29 = MaterialFunctionCall!(
   Diffuse 1
   Diffuse 2
   Transistion Phase=Subtract(A=@n22, B=LinearInterpolate(A=@n23, B=@n24, Alpha=@n25)) {constB=0.2}
   Height Texture 1=@n28
   Height Texture 2=@n26
   Contrast=ScalarParameter() {defaultValue=3, group=VertexPaintTexture, parameterName=VP_Outline_Contrast, sortPriority=6}
  ) {materialFunction=/Engine/Functions/Engine_MaterialFunctions02/Texturing/HeightLerpWithTwoHeightMaps}
  n30 = GetMaterialAttributes(None=MaterialFunctionCall!(UseLandscapeCoordsOnSwitch) {materialFunction=/Game/Athena/Environments/Landscape/MaterialFunctions/Snow/MF_Athena_Snow_01}.Material) {attributeGetTypes [#1]}
  n31 = VectorParameter() {group=VertexPaintTexture, parameterName=VP_Tint, sortPriority=13, defaultValue=<rgba 0.75 0.75 0.75 1>}
  n32 = ComponentMask(None=Multiply(
    A=StaticSwitchParameter(True=@n30.BaseColor, False=@n31.RGB) {group=VertexPaintTexture, parameterName=WinterSnow}
    B=ComponentMask(None=@n12) {b=true, g=true, r=true}
   )) {b=true, g=true, r=true}
  n33 = MaterialFunctionCall(
   Transistion Phase=@n22
   Height Texture 1=@n28
   Height Texture 2=@n26
   Contrast=ScalarParameter() {defaultValue=5, group=VertexPaintTexture, parameterName=VP_Contrast, sortPriority=5}
  ) {materialFunction=&HeightLerpWithTwoHeightMaps}
  n34 = Constant() {r=0.4}
  n35 = Constant() {r=0.6}
  n36 = Multiply(A=@n14.Roughness, B=ScalarParameter() {defaultValue=0.35, group=VertexPaintTexture, parameterName=VP_OutlineR_Mod, sortPriority=7})
  n37 = ScalarParameter() {group=VertexPaintTexture, parameterName=VPTexture_R_Min, sortPriority=8}
  n38 = ScalarParameter() {defaultValue=1, group=VertexPaintTexture, parameterName=VPTexture_R_Max, sortPriority=9}
  n39 = TextureObjectParameter() {
   group=VertexPaintTexture
   parameterName=VP_Spec
   samplerType=SAMPLERTYPE_Masks
   sortPriority=16
   texture=/Game/Environments/Sets/Pirate_Cove/Pirate_Ship/Textures/T_Pirate_Ship_Hull_B_S
  }
  n40 = TextureSample(UVs=@n1, Tex=@n39) {samplerType=SAMPLERTYPE_Masks, texture=&T_Pirate_Ship_Hull_B_S}
  n41 = Constant3Vector() {constant=<rgba 0 0 1 0>}
  n42 = TextureObjectParameter() {
   group=VertexPaintTexture
   parameterName=VP_Normal
   samplerType=SAMPLERTYPE_Normal
   sortPriority=16
   texture=/Game/Environments/Sets/Pirate_Cove/Pirate_Ship/Textures/T_Pirate_Ship_Hull_B_N
  }
  n43 = MaterialFunctionCall(
   Normal=TextureSample(UVs=@n9, Tex=@n42) {samplerType=SAMPLERTYPE_Normal, texture=&T_Snow01_N}.RGB
   Flatness=MaterialFunctionCall(
    1=ScalarParameter() {group=Material_1, parameterName=M1_VP_FlattenNormal, sortPriority=17}
    2=ScalarParameter() {group=Material_2, parameterName=M2_VP_FlattenNormal, sortPriority=17}
    3=ScalarParameter() {group=Material_3, parameterName=M3_VP_FlattenNormal, sortPriority=17}
    4=ScalarParameter() {group=Material_4, parameterName=M4_VP_FlattenNormal, sortPriority=17}
    Use 2 Textures=@n6
    Use 3 Textures=@n7
    Use 4 Textures=@n8
   ) {materialFunction=&MF_LerpTexturesBasedOnInt_Float4_4Layers}.Result
  ) {materialFunction=&FlattenNormal}
  n44 = StaticSwitchParameter(True=@n41, False=@n43.Result) {group=VertexPaintTexture, parameterName=FlatNormal, sortPriority=11}
  n45 = SetMaterialAttributes!(
   MaterialAttributes=@n14.MaterialAttributes
   BaseColor=LinearInterpolate(A=LinearInterpolate(A=@n14.BaseColor, B=@n16, Alpha=@n29.Alpha), B=@n32, Alpha=@n33.Alpha)
   Specular=LinearInterpolate(A=@n14.Specular, B=LinearInterpolate(A=@n34, B=@n35, Alpha=@n27), Alpha=@n33.Alpha)
   Roughness=LinearInterpolate(A=LinearInterpolate(A=@n14.Roughness, B=@n36, Alpha=@n29.Alpha), B=LinearInterpolate(A=@n37, B=@n38, Alpha=@n40.B), Alpha=@n33.Alpha)
   Normal=LinearInterpolate(A=@n14.Normal, B=QualitySwitch(Default=@n44, Low=@n41, High=@n44, Medium=@n44), Alpha=Multiply(A=@n33.Alpha))
   Metallic=LinearInterpolate(A=LinearInterpolate(A=@n14.Metallic, Alpha=@n29.Alpha) {constB=0.2}, Alpha=@n33.Alpha) {constB=0}
   EmissiveColor=LinearInterpolate(A=@n14.EmissiveColor, Alpha=@n33.Alpha) {constB=0}
  ) {attributeSetTypes [
    #1
    #2
    #3
    #4
    #5
    #6
   ]}
  FunctionOutput(None=StaticSwitchParameter(True=@n45, False=@n13) {desc=Red Paint Green Blur, group=VertexPaintTexture, parameterName="UseVP Texture (R Paint, G ActorNMap)"})
  n46 = ScalarParameter() {defaultValue=2}
  FunctionOutput(None=Multiply(
    A=MaterialFunctionCall(In=@n27, Contrast=@n46) {materialFunction=&CheapContrast}.Result
    B=@n33.Alpha
   )) {outputName=VP Height Mask}
  FunctionOutput(None=@n32) {outputName=VP_Color, sortPriority=2}
  n47 = Abs(None=VertexNormalWS())
  n48 = Multiply(A=ComponentMask(None=@n47) {r=true}, B=ComponentMask(None=@n47) {g=true})
  n49 = Constant() {r=3}
  LinearInterpolate(A=@n43.Result, B=@n41, Alpha=Saturate(None=Multiply(A=@n48, B=@n49)))
 ]}
