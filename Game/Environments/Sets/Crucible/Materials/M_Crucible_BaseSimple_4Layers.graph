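# Four-layer opaque base material for the Crucible environment set, with layers for interior walls, trims and doors.
# Each layer feeds tinted diffuse or DM color, specular masks, normals and emissive into a texture-set lerp; layer 1 can tint per floor.
# High quality allows parallax-occlusion UVs, with derivative mip sampling for normals.
# Vertex-paint texture and world-space gradient functions follow, bypassed on mobile, before the game-wide shared function.
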Material /Game/Environments/Sets/Crucible/Materials/M_Crucible_BaseSimple_4Layers
graph {
 properties {bUseMaterialAttributes=true, pixelDepthOffsetMode=PDOM_Legacy, usedShadingModels=Default Lit}
 nodes [
  n1 = MaterialFunctionCall() {materialFunction=/Game/Athena/WinterArt/Materials/MF_UVsSet}
  n2 = TextureObjectParameter() {group=Material_1, parameterName=Normals, samplerType=SAMPLERTYPE_Normal, sortPriority=2, texture=/Engine/EngineMaterials/FlatNormal}
  n3 = ScalarParameter() {defaultValue=1, group=Material_1, parameterName=Mat1 UV Scale, sortPriority=3}
  n4 = Multiply(A=TextureCoordinate(), B=Clamp(None=Ceil(None=@n3)) {maxDefault=1000, minDefault=1})
  n5 = TextureSample!(UVs=@n1.POM UVs 4 Layers, Tex=@n2, "DDX(UVs)"=DDX(Value=@n4), "DDY(UVs)"=DDY(Value=@n4), Apply View MipBias) {mipValueMode=TMVM_Derivative, samplerType=SAMPLERTYPE_Normal}
  n6 = TextureSample!(UVs=@n1.UVs Layer 1, Tex=@n2, Apply View MipBias) {samplerType=SAMPLERTYPE_Normal}
  n7 = QualitySwitch(
   Default=@n1.UVs Layer 1
   High=StaticSwitchParameter(True=@n1.POM UVs 4 Layers, False=@n1.POM UVs 4 Layers) {group=Material_1, parameterName=POM_True, sortPriority=11}
  )
  n8 = TextureObjectParameter() {
   group=Material_1
   parameterName=SpecularMasks
   samplerType=SAMPLERTYPE_Masks
   sortPriority=1
   texture=&T_Fortnite_Default_S
  }
  n9 = TextureSample(UVs=@n7, Tex=@n8) {samplerType=SAMPLERTYPE_Masks, texture=/Game/Environments/Sets/IceCastle/Textures/T_IceCastle_Wall1_D}
  n10 = MaterialFunctionCall() {materialFunction=&MF_UVsSet}
  n11 = QualitySwitch(
   Default=@n10.UVs Layer 2
   High=StaticSwitchParameter(True=@n10.POM UVs 4 Layers, False=@n10.POM UVs 4 Layers) {group=Material_2, parameterName=L2_POM_True, sortPriority=11}
  )
  n12 = TextureSampleParameter2D(UVs=@n11) {
   desc=Use for interior walls and ceilings
   group=Material_2
   parameterName=SpecularMasks_2
   samplerSource=SSM_Wrap_WorldGroupSettings
   samplerType=SAMPLERTYPE_Masks
   sortPriority=2
   texture=&T_Fortnite_Default_S
  }
  n13 = TextureSampleParameter2D(UVs=@n10.UVs Layer 2) {
   desc=Use for interior walls and ceilings
   group=Material_2
   parameterName=Normals_Texture_2
   samplerSource=SSM_Wrap_WorldGroupSettings
   samplerType=SAMPLERTYPE_Normal
   sortPriority=3
   texture=&FlatNormal
  }
  n14 = TextureSampleParameter2D!(UVs=@n10.POM UVs 4 Layers, "DDX(UVs)"=DDX(Value=@n10.UVs Layer 2), "DDY(UVs)"=DDY(Value=@n10.UVs Layer 2), Apply View MipBias) {
   desc=Use for interior walls and ceilings
   group=Material_2
   mipValueMode=TMVM_Derivative
   parameterName=Normals_Texture_2
   samplerSource=SSM_Wrap_WorldGroupSettings
   samplerType=SAMPLERTYPE_Normal
   sortPriority=3
   texture=&FlatNormal
  }
  n15 = TextureObjectParameter() {
   group=Material_3
   parameterName=DM_3
   samplerType=SAMPLERTYPE_Masks
   sortPriority=1
   texture=&T_Fortnite_Default_S
  }
  n16 = TextureSample(Tex=@n15) {samplerType=SAMPLERTYPE_Masks}
  n17 = VectorParameter() {group=Material_3, parameterName=M3_TintColor, sortPriority=5, defaultValue=<rgba 1 1 1 0>}
  n18 = MaterialFunctionCall!(
   DM=@n16.RGB
   Color1=@n17.RGB
   Color2=VectorParameter() {group=Material_3, parameterName=M3_TintColor2, sortPriority=5, defaultValue=<rgba 1 1 1 0>}.RGB
   Color3=VectorParameter() {group=Material_3, parameterName=M3_TintColor3, sortPriority=5, defaultValue=<rgba 1 1 1 0>}.RGB
  ) {materialFunction=/Game/Athena/WinterArt/Materials/MF_DMTexture}
  n19 = MaterialFunctionCall() {materialFunction=&MF_UVsSet}
  n20 = QualitySwitch(
   Default=@n19.UVs Layer 3
   High=StaticSwitchParameter(True=@n19.POM UVs 4 Layers, False=@n19.POM UVs 4 Layers) {group=Material_3, parameterName=L3_POM_True, sortPriority=11}
  )
  n21 = TextureSampleParameter2D!(UVs=@n20, Apply View MipBias) {
   desc=Use for trims and accents
   group=Material_3
   parameterName=Diffuse_Texture_3
   samplerSource=SSM_Wrap_WorldGroupSettings
   sortPriority=1
   texture=/Game/Global/Textures/Default/Blanks/T_MidgreysRGB
  }
  n22 = Multiply(A=MaterialFunctionCall!(In=@n17.RGB) {materialFunction=/Game/Athena/WinterArt/Materials/MF_VPTintMask}.Result, B=@n21.RGB)
  n23 = TextureObjectParameter() {
   group=Material_4
   parameterName=DM_4
   samplerType=SAMPLERTYPE_Masks
   sortPriority=1
   texture=&T_Fortnite_Default_S
  }
  n24 = TextureSample(Tex=@n23) {samplerType=SAMPLERTYPE_Masks}
  n25 = VectorParameter() {group=Material_4, parameterName=M4_TintColor, sortPriority=5, defaultValue=<rgba 1 1 1 0>}
  n26 = MaterialFunctionCall(
   DM=@n24.RGB
   Color1=@n25.RGB
   Color2=VectorParameter() {group=Material_4, parameterName=M4_TintColor2, sortPriority=5, defaultValue=<rgba 1 1 1 0>}.RGB
   Color3=VectorParameter() {group=Material_4, parameterName=M4_TintColor3, sortPriority=5, defaultValue=<rgba 1 1 1 0>}.RGB
  ) {materialFunction=&MF_DMTexture}
  n27 = MaterialFunctionCall() {materialFunction=&MF_UVsSet}
  n28 = QualitySwitch(
   Default=@n27.UVs Layer 4
   High=StaticSwitchParameter(True=@n27.POM UVs 4 Layers, False=@n27.POM UVs 4 Layers) {group=Material_4, parameterName=L4_POM_True, sortPriority=11}
  )
  n29 = TextureSampleParameter2D(UVs=@n28) {
   desc=Use for doors
   group=Material_4
   parameterName=Diffuse_Texture_4
   samplerSource=SSM_Wrap_WorldGroupSettings
   sortPriority=1
   texture=&T_MidgreysRGB
  }
  n30 = Multiply(A=MaterialFunctionCall(In=@n25.RGB) {materialFunction=&MF_VPTintMask}.Result, B=@n29.RGB)
  n31 = TextureObjectParameter() {
   group=Material_2
   parameterName=DM_2
   samplerType=SAMPLERTYPE_Masks
   sortPriority=1
   texture=&T_Fortnite_Default_S
  }
  n32 = TextureSample(UVs=@n11, Tex=@n31) {samplerType=SAMPLERTYPE_Masks}
  n33 = VectorParameter() {group=Material_2, parameterName=M2_TintColor, sortPriority=5, defaultValue=<rgba 1 1 1 0>}
  n34 = MaterialFunctionCall(
   DM=@n32.RGB
   Color1=@n33.RGB
   Color2=VectorParameter() {group=Material_2, parameterName=M2_TintColor2, sortPriority=5, defaultValue=<rgba 1 1 1 0>}.RGB
   Color3=VectorParameter() {group=Material_2, parameterName=M2_TintColor3, sortPriority=5, defaultValue=<rgba 1 1 1 0>}.RGB
  ) {materialFunction=&MF_DMTexture}
  n35 = TextureSampleParameter2D(UVs=@n11) {
   desc=Use for interior walls and ceilings
   group=Material_2
   parameterName=Diffuse_Texture_2
   samplerSource=SSM_Wrap_WorldGroupSettings
   sortPriority=1
   texture=&T_MidgreysRGB
  }
  n36 = Multiply(A=MaterialFunctionCall(In=@n33.RGB) {materialFunction=&MF_VPTintMask}.Result, B=@n35.RGB)
  n37 = TextureSampleParameter2D(UVs=@n20) {
   desc=Use for trims and accents
   group=Material_3
   parameterName=SpecularMasks_3
   samplerSource=SSM_Wrap_WorldGroupSettings
   samplerType=SAMPLERTYPE_Masks
   sortPriority=2
   texture=&T_Fortnite_Default_S
  }
  n38 = TextureSampleParameter2D(UVs=@n19.UVs Layer 3) {
   desc=Use for trims and accents
   group=Material_3
   parameterName=Normals_Texture_3
   samplerSource=SSM_Wrap_WorldGroupSettings
   samplerType=SAMPLERTYPE_Normal
   sortPriority=3
   texture=&FlatNormal
  }
  n39 = TextureSampleParameter2D!(UVs=@n19.POM UVs 4 Layers, "DDX(UVs)"=DDX(Value=@n19.UVs Layer 3), "DDY(UVs)"=DDY(Value=@n19.UVs Layer 3), Apply View MipBias) {
   desc=Use for trims and accents
   group=Material_3
   mipValueMode=TMVM_Derivative
   parameterName=Normals_Texture_3
   samplerSource=SSM_Wrap_WorldGroupSettings
   samplerType=SAMPLERTYPE_Normal
   sortPriority=3
   texture=&FlatNormal
  }
  n40 = TextureSampleParameter2D!(UVs=@n28, Apply View MipBias) {
   desc=Use for doors
   group=Material_4
   parameterName=SpecularMasks_4
   samplerSource=SSM_Wrap_WorldGroupSettings
   samplerType=SAMPLERTYPE_Masks
   sortPriority=2
   texture=&T_Fortnite_Default_S
  }
  n41 = TextureSampleParameter2D() {
   desc=Use for doors
   group=Material_4
   parameterName=Normals_Texture_4
   samplerSource=SSM_Wrap_WorldGroupSettings
   samplerType=SAMPLERTYPE_Normal
   sortPriority=3
   texture=&FlatNormal
  }
  n42 = TextureSampleParameter2D!(UVs=@n27.POM UVs 4 Layers, "DDX(UVs)"=DDX(Value=@n27.UVs Layer 4), "DDY(UVs)"=DDY(Value=@n27.UVs Layer 4), Apply View MipBias) {
   desc=Use for doors
   group=Material_4
   mipValueMode=TMVM_Derivative
   parameterName=Normals_Texture_4
   samplerSource=SSM_Wrap_WorldGroupSettings
   samplerType=SAMPLERTYPE_Normal
   sortPriority=3
   texture=&FlatNormal
  }
  n43 = TextureObjectParameter() {
   group=Material_1
   parameterName=DM
   samplerType=SAMPLERTYPE_Masks
   sortPriority=1
   texture=&T_Fortnite_Default_S
  }
  n44 = TextureSample(UVs=@n7, Tex=@n43) {samplerType=SAMPLERTYPE_Masks}
  n45 = VectorParameter() {group=Material_1, parameterName=TintColor, sortPriority=7, defaultValue=<rgba 1 1 1 0>}
  n46 = MaterialFunctionCall(
   DM=@n44.RGB
   Color1=@n45.RGB
   Color2=VectorParameter() {group=Material_1, parameterName=M1_TintColor2, sortPriority=5, defaultValue=<rgba 1 1 1 0>}.RGB
   Color3=VectorParameter() {group=Material_1, parameterName=M1_TintColor3, sortPriority=5, defaultValue=<rgba 1 1 1 0>}.RGB
  ) {materialFunction=&MF_DMTexture}
  n47 = TextureObjectParameter() {group=Material_1, parameterName=Diffuse, texture=&T_IceCastle_Wall1_D}
  n48 = TextureSample(UVs=@n7, Tex=@n47) {texture=&T_IceCastle_Wall1_D}
  n49 = MaterialFunctionCall!(
   Floor1=@n45.RGB
   Floor2=VectorParameter() {group=Material_1, parameterName=Floor2, sortPriority=8, defaultValue=<rgba 1 1 1 0>}.RGB
   Floor3=VectorParameter() {group=Material_1, parameterName=Floor3, sortPriority=9, defaultValue=<rgba 1 1 1 0>}.RGB
   Basement=VectorParameter() {group=Material_1, parameterName=Basement, sortPriority=10, defaultValue=<rgba 1 1 1 0>}.RGB
  ) {materialFunction=/Game/Athena/WinterArt/Materials/MF_WS_GridUV}
  n50 = Multiply(
   A=@n48.RGB
   B=MaterialFunctionCall(In=StaticSwitchParameter(True=@n49.StoryMaskOut, False=@n45.RGB) {group=Material_1, parameterName=UsePerFloorColor, sortPriority=4}) {materialFunction=&MF_VPTintMask}.Result
  )
  n51 = MaterialFunctionCall!(
   Use3Materials
   Use4Materials
   Use2Materials
   D=StaticSwitchParameter(True=AppendVector(A=@n46.Result, B=@n44.A), False=AppendVector(A=@n50, B=@n48.A)) {desc=Creative, group=Material_1, parameterName=LM_DMTextureTrue, sortPriority=11}
   S=AppendVector(A=@n9.RGB, B=@n9.RGB)
   N=QualitySwitch(Default=@n6.RGB, High=StaticSwitchParameter(True=@n5.RGB, False=@n6.RGB) {group=Material_1, parameterName=POM_True, sortPriority=11})
   D2=StaticSwitchParameter(True=AppendVector(A=@n34.Result, B=@n32.A), False=AppendVector(A=@n36, B=@n35.A)) {desc=Creative, group=Material_2, parameterName=L2_DMTextureTrue, sortPriority=11}
   S2=AppendVector(A=@n12.RGB, B=@n12.RGB)
   N2=QualitySwitch(Default=@n13.RGB, High=StaticSwitchParameter(True=@n14.RGB, False=@n13.RGB) {group=Material_2, parameterName=L2_POM_True, sortPriority=11})
   D3=StaticSwitchParameter(True=AppendVector(A=@n18.Result, B=@n16.A), False=AppendVector(A=@n22, B=@n21.A)) {desc=Creative, group=Material_3, parameterName=L3_DMTextureTrue, sortPriority=11}
   S3=AppendVector(A=@n37.RGB, B=@n37.RGB)
   N3=QualitySwitch(Default=@n38.RGB, High=StaticSwitchParameter(True=@n39.RGB, False=@n38.RGB) {group=Material_3, parameterName=L3_POM_True, sortPriority=11})
   D4=StaticSwitchParameter(True=AppendVector(A=@n26.Result, B=@n24.A), False=AppendVector(A=@n30, B=@n29.A)) {desc=Creative, group=Material_3, parameterName=L3_DMTextureTrue, sortPriority=11}
   S4=AppendVector(A=@n40.RGB, B=@n40.RGB)
   N4=QualitySwitch(Default=@n41.RGB, High=StaticSwitchParameter(True=@n42.RGB, False=@n41.RGB) {group=Material_4, parameterName=L4_POM_True, sortPriority=11})
   E1=VectorParameter() {group=Material_1, parameterName=Emissive, sortPriority=7}.RGB
   E2=VectorParameter() {group=Material_2, parameterName=Emissive2, sortPriority=7}.RGB
   E3=VectorParameter() {group=Material_3, parameterName=Emissive3, sortPriority=7}.RGB
   E4=VectorParameter() {group=Material_4, parameterName=Emissive4, sortPriority=7}.RGB
   CCMaskByMetal_1
   CC1
   CCR_1
   CC2
   CCR_2
   CCMaskByMetal_2
   CC3
   CCR_3
   CCMaskByMetal_3
   CC4
   CCMaskByMetal_4
   CCR_4
   OpacityMaskCancel_1
   OpacityMaskCancel_2
   OpacityMaskCancel_3
   OpacityMaskCancel_4
   SSS_1
   SSS_2
   SSS_3
   SSS_4
   SSS_Depth_1
   SSS_Depth_2
   SSS_Depth_3
   SSS_Depth_4
  ) {materialFunction=/Game/Athena/WinterArt/Materials/MF_FN_TextureSetLerp_4Layers}
  n52 = SetMaterialAttributes!(
   MaterialAttributes
   Base Color=@n51.Diffuse
   Specular=@n51.Diffuse
   Metallic=@n51.Diffuse
   Roughness=@n51.Diffuse
   Emissive Color=@n51.Diffuse
   Normal=@n51.Diffuse
  ) {attributeSetTypes [
    #1
    #2
    #3
    #4
    #5
    #6
   ]}
  n53 = MaterialFunctionCall!(transitionFactor, 1, In=@n52, 2, BGMap=@n51.Height, 3, 4) {materialFunction=/Game/Environments/Sets/Pirate_Cove/Pirate_Ship/Materials/MaterialFunctions/MF_Pirate_Ship_VPTexture}
  n54 = MaterialFunctionCall!(transitionFactor, VP mask=@n53.VP Height Mask, 1, MaterialIn=@n53.VP Height Mask, 2, 3, 4) {materialFunction=&/MF_Crucible_WSGradient1}
  MaterialFunctionCall!(In=ShadingPathSwitch(Default=@n54.Color*Gradient, Deferred=@n54.Color*Gradient, Mobile=@n52) {desc=Migrated from FeatureLevelSwitch}) {materialFunction=/Game/Athena/WinterArt/Materials/MF_GameWideShared}
  SetMaterialAttributes() {attributeSetTypes [
    #1
    #2
    #3
    #4
    #5
    #6
   ]}
  n55 = TextureObjectParameter() {group=Material_2, parameterName=Diffuse_Texture_2, sortPriority=1, texture=&T_MidgreysRGB}
  StaticSwitchParameter(True=@n31, False=@n55) {desc=Creative, group=Material_2, parameterName=L2_DMTextureTrue, sortPriority=11}
  n56 = TextureObjectParameter() {group=Material_4, parameterName=Diffuse_Texture_4, sortPriority=1, texture=&T_MidgreysRGB}
  StaticSwitchParameter(True=@n23, False=@n56) {desc=Creative, group=Material_3, parameterName=L3_DMTextureTrue, sortPriority=11}
  StaticSwitchParameter(True=@n43, False=@n47) {desc=Creative, group=Material_1, parameterName=L1_DMTextureTrue, sortPriority=11}
  n57 = TextureObjectParameter() {group=Material_3, parameterName=Diffuse_Texture_3, sortPriority=1, texture=&T_MidgreysRGB}
  StaticSwitchParameter(True=@n15, False=@n57) {desc=Creative, group=Material_3, parameterName=L3_DMTextureTrue, sortPriority=11}
 ]
}
